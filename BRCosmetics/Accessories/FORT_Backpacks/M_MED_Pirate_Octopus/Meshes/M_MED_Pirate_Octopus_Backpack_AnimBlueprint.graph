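# Animation Blueprint for the Pirate Octopus backpack cosmetic, animating its chest lid and dangling necklace pieces.
# The event graph sets close_lid when the pawn moves fast horizontally or moves upward, and clears it otherwise.
# That value drives an interpolated alpha on an additive lid-bone rotation, starting from the ref pose.
# Chained AnimDynamics nodes then add angle-limited secondary motion to the lid and three necklace bones, using overridden gravity and damped component motion.
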AnimBlueprint /BRCosmetics/Accessories/FORT_Backpacks/M_MED_Pirate_Octopus/Meshes/M_MED_Pirate_Octopus_Backpack_AnimBlueprint
target_skeleton=&/M_MED_Pirate_Octopus_Backpack_Skeleton
graphs [
 AnimGraph {nodes [
   n1 = "Transform (Modify) Bone - Bone: dyn_boxCover"(
    ComponentPose=Local To Component(LocalPose=Local Space Ref Pose())
    Alpha=Get close_lid() -> (close_lid:double)
    Translation="0,0,0"
    Rotation="0,0,-26.66"
    Scale="1,1,1"
   ) {
    rotationMode=BMM_Additive
    alphaBoolBlend {blendOption=Linear}
    alphaScaleBiasClamp {bInterpResult=true, interpSpeedIncreasing=1, scale=5}
    boneToModify {boneName=dyn_boxCover}
    rotation {roll=-26.66}
   }
   n2 = Anim Dynamics - Bone: dyn_boxCover(ComponentPose=@n1, Alpha=1, ExternalForce="0,0,0") {
    angularDampingOverride=0.99
    bGravityOverrideInSimSpace=true
    bOverrideAngularDamping=true
    bUseGravityOverride=true
    alphaBoolBlend {blendOption=Linear}
    boundBone {boneName=dyn_boxCover}
    componentAppliedLinearAccClamp=<1000 1000 1000>
    componentLinearAccScale=<0.3 0.3 0.3>
    componentLinearVelScale=<0.3 0.3 0.3>
    gravityOverride {z=-980}
    physicsBodyDefinitions [{
      sphereCollisionRadius=0
      boundBone {boneName=dyn_boxCover}
      boxExtents=<15 5 10>
      constraintSetup {angularLimitsMax=<0 0 35>, angularLimitsMin=<0 0 -5>, angularTarget=<0 0 0>}
      localJointOffset=<0.9 0 0>
     }]
   }
   n3 = Anim Dynamics - Bone: dyn_necklace_02(ComponentPose=@n2, Alpha=1, ExternalForce="0,0,0") {
    angularDampingOverride=0.8
    bGravityOverrideInSimSpace=true
    bOverrideAngularDamping=true
    bUseGravityOverride=true
    alphaBoolBlend {blendOption=Linear}
    boundBone {boneName=dyn_necklace_02}
    componentAppliedLinearAccClamp=<1000 1000 1000>
    componentLinearAccScale=<0.5 0.5 0.5>
    componentLinearVelScale=<0.5 0.5 0.5>
    gravityOverride {z=-980}
    physicsBodyDefinitions [{
      sphereCollisionRadius=0
      boundBone {boneName=dyn_necklace_02}
      boxExtents=<6 4 4>
      constraintSetup {angularLimitsMax=<0 0 45>, angularLimitsMin=<0 0 0>, angularTarget=<0 0 0>}
      localJointOffset=<1 0 0>
     }]
   }
   n4 = Anim Dynamics - Bone: dyn_necklace_03(ComponentPose=@n3, Alpha=1, ExternalForce="0,0,0") {
    angularDampingOverride=0.8
    bGravityOverrideInSimSpace=true
    bOverrideAngularDamping=true
    bUseGravityOverride=true
    alphaBoolBlend {blendOption=Linear}
    boundBone {boneName=dyn_necklace_03}
    componentAppliedLinearAccClamp=<1000 1000 1000>
    componentLinearAccScale=<0.5 0.5 0.5>
    componentLinearVelScale=<0.5 0.5 0.5>
    gravityOverride {z=-980}
    physicsBodyDefinitions [{
      sphereCollisionRadius=0
      boundBone {boneName=dyn_necklace_03}
      boxExtents=<6 4 4>
      constraintSetup {angularLimitsMax=<0 15 80>, angularLimitsMin=<0 -15 0>, angularTarget=<0 0 0>}
      localJointOffset=<1 0 0>
     }]
   }
   n5 = Anim Dynamics - Bone: dyn_necklace_01(ComponentPose=@n4, Alpha=1, ExternalForce="0,0,0") {
    angularDampingOverride=0.8
    bGravityOverrideInSimSpace=true
    bOverrideAngularDamping=true
    bUseGravityOverride=true
    alphaBoolBlend {blendOption=Linear}
    boundBone {boneName=dyn_necklace_01}
    componentAppliedLinearAccClamp=<1000 1000 1000>
    componentLinearAccScale=<0.5 0.5 0.5>
    componentLinearVelScale=<0.5 0.5 0.5>
    gravityOverride {z=-980}
    physicsBodyDefinitions [{
      sphereCollisionRadius=0
      boundBone {boneName=dyn_necklace_01}
      boxExtents=<6 4 4>
      constraintSetup {angularLimitsMax=<0 15 80>, angularLimitsMin=<0 -15 0>, angularTarget=<0 0 0>}
      localJointOffset=<1 0 0>
     }]
   }
   Output Pose(Result=Component To Local(ComponentPose=@n5)) {layerGroup=None}
  ]}
 EventGraph {nodes [
   Event BlueprintUpdateAnimation()
   n1 = Get PawnSpeed2D()
   n2 = Get PawnVelocityZ()
   > n3 = Branch(Condition=OR Boolean(A="float > float"(A=@n1, B=481) {is=title K2Node_CallFunction}, B="float > float"(A=@n2, B=0)))
   > Set close_lid(close_lid:double=1) -> (Output_Get:double)
   Set close_lid(execute=@n3.else, close_lid=0)
  ]}
]
variables [close_lid {type=number}]
