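# Physics asset for the pirate octopus backpack cosmetic accessory.
# It holds a single body on the backpack_root bone, shaped as one offset capsule rotated a quarter turn from its default axis.
# There are no constraints and no other bones have bodies, so the whole backpack is approximated by that one collision volume.
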PhysicsAsset /BRCosmetics/Accessories/FORT_Backpacks/M_MED_Pirate_Octopus/Meshes/M_MED_Pirate_Octopus_Backpack_Physics
bodies {backpack_root [{shapeType=Capsule, length=14.7606, radius=16.5959, center=<-2.3056 -11.5109 0>, rotation=<rot 0 -90 0>}]}
constraints []
